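# Looping ambient spot cue that plays occasional long wood creaks for Kapok trees at the Asteria POI, routed as outdoor one-shots.
# An endless loop waits a random delay, then applies light pitch and volume variation before a quality switch.
# The two higher-index quality branches pick from four creaks instead of sixteen, each pool adding wider pitch and volume variation.
# Attenuation, a single-voice concurrency override and an exterior ambient bus send shape playback.
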SoundCue /Game/Sounds/Asteria/POI/KapokTrees/Spots/Spot_Kapok_Tree_Creak_Long_Looping_Cue
root=Looping_0
nodes [
 {id=Delay_0, children [Modulator_13], properties {delayMax=18, delayMin=12}}
 {id=Looping_0, children [Delay_0]}
 {id=Modulator_0, children [Random_1], properties {pitchMax=1, pitchMin=0.7, volumeMax=1.1, volumeMin=0.5}}
 {id=Modulator_1, children [Random_0], properties {pitchMax=1, pitchMin=0.7, volumeMax=1.1, volumeMin=0.5}}
 {id=Modulator_13, children [QualityLevel_0], properties {pitchMax=1.1, pitchMin=0.9, volumeMax=1.1, volumeMin=0.9}}
 {id=QualityLevel_0, children [Modulator_1,Modulator_1,Modulator_0,Modulator_0]}
 {
  id=Random_0
  children [
   WavePlayer_0
   WavePlayer_1
   WavePlayer_2
   WavePlayer_3
   WavePlayer_4
   WavePlayer_5
   WavePlayer_6
   WavePlayer_7
   WavePlayer_8
   WavePlayer_9
   WavePlayer_10
   WavePlayer_11
   WavePlayer_12
   WavePlayer_13
   WavePlayer_14
   WavePlayer_15
  ]
  properties {weights [1,1,1,1,1,1,1,1,1,1,1,1,1,1,1,1]}
 }
 {id=Random_1, children [WavePlayer_16, WavePlayer_17, WavePlayer_18, WavePlayer_19], properties {weights [1,1,1,1]}}
 {id=WavePlayer_0, properties {soundWaveAssetPtr=/Game/Sounds/Trans_AmbientNonLoop/Wood/TreeCreakLong001}}
 {id=WavePlayer_1, properties {soundWaveAssetPtr=/Game/Sounds/Trans_AmbientNonLoop/Wood/TreeCreakLong002}}
 {id=WavePlayer_10, properties {soundWaveAssetPtr=/Game/Sounds/Trans_AmbientNonLoop/Wood/TreeCreakLong011}}
 {id=WavePlayer_11, properties {soundWaveAssetPtr=/Game/Sounds/Trans_AmbientNonLoop/Wood/TreeCreakLong012}}
 {id=WavePlayer_12, properties {soundWaveAssetPtr=/Game/Sounds/Trans_AmbientNonLoop/Wood/TreeCreakLong013}}
 {id=WavePlayer_13, properties {soundWaveAssetPtr=/Game/Sounds/Trans_AmbientNonLoop/Wood/TreeCreakLong014}}
 {id=WavePlayer_14, properties {soundWaveAssetPtr=/Game/Sounds/Trans_AmbientNonLoop/Wood/TreeCreakLong015}}
 {id=WavePlayer_15, properties {soundWaveAssetPtr=/Game/Sounds/Trans_AmbientNonLoop/Wood/TreeCreakLong016}}
 {id=WavePlayer_16, properties {soundWaveAssetPtr=&TreeCreakLong015}}
 {id=WavePlayer_17, properties {soundWaveAssetPtr=/Game/Sounds/Trans_AmbientNonLoop/Wood/TreeCreakLong005}}
 {id=WavePlayer_18, properties {soundWaveAssetPtr=&TreeCreakLong013}}
 {id=WavePlayer_19, properties {soundWaveAssetPtr=&TreeCreakLong011}}
 {id=WavePlayer_2, properties {soundWaveAssetPtr=/Game/Sounds/Trans_AmbientNonLoop/Wood/TreeCreakLong003}}
 {id=WavePlayer_3, properties {soundWaveAssetPtr=/Game/Sounds/Trans_AmbientNonLoop/Wood/TreeCreakLong004}}
 {id=WavePlayer_4, properties {soundWaveAssetPtr=&TreeCreakLong005}}
 {id=WavePlayer_5, properties {soundWaveAssetPtr=/Game/Sounds/Trans_AmbientNonLoop/Wood/TreeCreakLong006}}
 {id=WavePlayer_6, properties {soundWaveAssetPtr=/Game/Sounds/Trans_AmbientNonLoop/Wood/TreeCreakLong007}}
 {id=WavePlayer_7, properties {soundWaveAssetPtr=/Game/Sounds/Trans_AmbientNonLoop/Wood/TreeCreakLong008}}
 {id=WavePlayer_8, properties {soundWaveAssetPtr=/Game/Sounds/Trans_AmbientNonLoop/Wood/TreeCreakLong009}}
 {id=WavePlayer_9, properties {soundWaveAssetPtr=/Game/Sounds/Trans_AmbientNonLoop/Wood/TreeCreakLong010}}
]
properties {
 attenuationSettings=&/Spot_Kopok_Tree_Creak_Actor_Att
 bOverrideConcurrency=true
 maxDistance=4500
 soundClassObject=&Fort_Ambient_OneShots_Outdoors
 volumeMultiplier=1.504
 busSends [{
   audioBus=null
   maxSendLevel=0.95
   minSendLevel=0.3
   sendLevel=0
   soundSourceBus=&AmbientAudio_Exterior_SourceBus
  }]
 concurrencyOverrides {maxCount=1, voiceStealReleaseTime=3}
 preEffectBusSends [{}]
}
